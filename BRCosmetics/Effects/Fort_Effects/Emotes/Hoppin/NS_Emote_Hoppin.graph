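# Niagara effect for the Hoppin emote that spawns short-lived, numbered hopscotch tiles at the character's feet.
# User HopscotchSpawnCount and a system-computed DT_HopscotchCount drive spawning, while HopscotchCount picks each tile's sub-image.
# FeetCurve and spawn index pick bone and color, with tiles pinned flat near simulation-origin height using separate front-end and gameplay offsets.
# Two sprite renderers draw each particle, which flashes bright then fades through color and material-parameter curves.
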NiagaraSystem /BRCosmetics/Effects/Fort_Effects/Emotes/Hoppin/NS_Emote_Hoppin
dependencies {
 usedDataInterfaces [
  &NiagaraDataInterfaceArrayColor
  &NiagaraDataInterfaceSkeletalMesh
  &NiagaraDataInterfaceArrayFloat3
  &NiagaraDataInterfaceCurve
 ]
 usedDynamicInputs []
 usedModules [
  &SystemState
  &self:SystemSpawnScript.SystemScriptSource.SystemScriptGraph.NiagaraNodeAssignment_0.SetVariables_#1
  &self:DT_HopscotchCount
 ]
}
emitters {Tiles001 {
  emitterUpdateScript [
   &EmitterState()
   &self:ScratchModule_01(Count:NiagaraInt32=&MakeIntFromFloat(Float=&Multiply_Float(A=$System.DT_HopscotchCount, B=$User.HopscotchSpawnCount))) {module=ScratchModule_01}
  ]
  particleSpawnScript [
   &InitializeParticle(
    Color=&SelectLinearColorFromArray(
     Array Sampling Mode=::NewEnumerator1
     Direct Array Index=&Add_Integer(A=&ReturnExecIndex(), B=&MakeIntFromFloat(Float=$User.FeetCurve))
     Direct Array Mode=::NewEnumerator1
     Color Selection Array=DataInterface() {propertyValues {ColorData [<rgba 0.9068 0.2 2 1>, <rgba 1 0.398 0.1038 1>]}}
     Random Seed=?
    )
    Lifetime=0.7
    Mesh Scale Mode=::NewEnumerator3
    Position Offset=&Lerp_Vector(A=<0 0 0>, B=<0 0 0>, Alpha=0.5)
    Sprite Rotation Mode=::NewEnumerator1
    Sprite Size Mode=::NewEnumerator3
    Uniform Sprite Size=35
    UsePositionOffset=false
    Write Material Random=true
   ) {module=InitializeParticle001}
   /SetVariables(Particles.HopscotchSquare:Vector2f=&Vector2DFromFloat(Value=50)) {isSetParameters=true}
   /SetVariables(Particles.HopscotchSquareColor:LinearColor=<rgba 0.4 0.4 0.4 1>) {isSetParameters=true}
   /SetVariables(Particles.SubImageIndex=&Add_Float(A=$User.HopscotchCount, B=&ReturnNormalizedExecIndex())) {isSetParameters=true}
   /SetVariables(Particles.SpriteAlignment=$Engine.Owner.SystemYAxis, Particles.SpriteFacing=<0 0 1>) {isSetParameters=true}
   &SkeletalMeshLocation(
    Bone / Socket Index=&MakeCustomIntFromBool(
     Bool=&SetBoolByFloatComparison(A=$User.HopscotchSpawnCount, B=1)
     True INT=3
     False INT=&Add_Integer(A=&ReturnExecIndex(), B=&MakeIntFromFloat(Float=$User.FeetCurve))
    )
    Bone / Socket Sampling Mode=::NewEnumerator5
    Interpolate Mesh Sampling=false
    MaskSpawnGroup=false
    Mesh Sampling Type=::NewEnumerator6
    Sampled Position Coordinate Space=::World
    Sampled Position Offset=&MakeCustomVectorFromBool(
     Bool=&SetBoolByFloatComparison(A=$User.HopscotchSpawnCount, B=1)
     True Vector=&FN_FrontEndVector3Modifier(
      Gameplay=&SelectVectorFromArray(
       Direct Array Index=&ReturnExecIndex()
       Vector Selection Array=DataInterface() {propertyValues {FloatData [<21.5 0 0>, <-20 0 0>]}}
      )
      Front End=&SelectVectorFromArray(
       Direct Array Index=&ReturnExecIndex()
       Vector Selection Array=DataInterface() {propertyValues {FloatData [<21.5 10 0>, <-20 10 0>]}}
      )
     )
     False Vector=&FN_FrontEndVector3Modifier(Gameplay=<0 0 0>, Front End=<0 10 0>)
    )
    Sampled Position Offset Coordinate Space:ENiagaraSkelMeshTransforms=::NewEnumerator2
    Skeletal Mesh=DataInterface() {propertyValues {
      DefaultMesh=null
      ExcludeBoneName=root
      OptionalComponentClass=null
      PreviewMesh=/SaveTheWorld/Characters/NPCs/Survivors/M_SML_Survivor_01/Meshes/SK_M_SML_Survivor_01_Preview
      SkinningMode=SkinOnTheFly
      SoftSourceActor=null
      SourceMode=Default
      UvSetIndex=0
      WholeMeshLOD=-1
      bExcludeBone=false
      bReadDeformedGeometry=true
      bRequireCurrentFrameData=true
      ComponentTags []
      FilteredBones [foot_l,foot_r,root]
      FilteredSockets []
      MeshUserParameter {parameter {name=None, typeDefHandle {registeredTypeIndex=1}, varData []}}
      SamplingRegions []
     }}
   )
   /SetVariables(Particles.Position=&MakeVector(
     X=&MakeFloatFromVector(Vector=Linked() {linkedVariable {name=Particles.Position, type=NiagaraPosition}}, Channel=::NewEnumerator0)
     Y=&MakeFloatFromVector(Vector=Linked() {linkedVariable {name=Particles.Position, type=NiagaraPosition}}, Channel=::NewEnumerator1)
     Z=&MakeFloatFromVector(
      Vector=&Add_Vector(
       A=/Niagara/DynamicInputs/Helpers/SimulationPosition()
       B=&FN_FrontEndVector3Modifier(Gameplay=<0 0 5>, Front End=<0 0 2>)
      )
      Channel=::NewEnumerator2
     )
    )) {isSetParameters=true}
   &DynamicMaterialParameters(
    Index 2 Param 1=0.5
    Index 2 Param 2=0.5
    Index 2 Param 3=&RandomRangeFloat(Minimum=0, Maximum=1)
    Index 2 Param 4=&RandomRangeFloat(Minimum=0, Maximum=1)
    Module.Param0WriteEnabled2=true
    Module.Param1WriteEnabled2=true
    Module.Param2WriteEnabled2=true
    Module.Param3WriteEnabled2=true
    Write Parameter Index 0=false
    Write Parameter Index 2=true
   ) {module=DynamicMaterialParameters001}
  ]
  particleUpdateScript [
   &ParticleState()
   &SolveForcesAndVelocity(Clamp Velocity=false, Limit Acceleration=false) {enabled=false}
   &self:DistanceBased_Scaling(Distance=50, Inside Distance=true, ScaleSize:Vector3f=<0.75 0.75 0.1>) {enabled=false, module=DistanceBased_Scaling}
   &ScaleColor(
    Scale Alpha=&FloatFromCurve(FloatCurve=DataInterface() {propertyValues {Curve {keys [<at 0 1>,<at 1 0>]}}})
    Scale RGB=&VectorFromFloat(Value=&FloatFromCurve(FloatCurve=DataInterface() {propertyValues {Curve {keys [
          {interpMode=RCIM_Cubic, tangentMode=RCTM_User, time=0, value=0.5}
          {interpMode=RCIM_Cubic, tangentMode=RCTM_User, time=0.05, value=5}
          {interpMode=RCIM_Cubic, tangentMode=RCTM_User, time=0.15, value=1}
          {interpMode=RCIM_Cubic, tangentMode=RCTM_User, time=1, value=0}
         ]}}}))
    ScaleA=false
    ScaleRGB=true
   )
   &DynamicMaterialParameters(
    Index 0 Param 1=1
    Index 0 Param 2=1
    Index 3 Param 1=0.8
    Index 3 Param 2=&FloatFromCurve(
     FloatCurve=DataInterface() {propertyValues {Curve {keys [
         <at 0 1>
         {arriveTangent=2.8201, interpMode=RCIM_Cubic, leaveTangent=2.8201, tangentMode=RCTM_User, time=0.1, value=0}
         {arriveTangent=-0.0405, interpMode=RCIM_Cubic, leaveTangent=-0.0405, tangentMode=RCTM_User, time=1, value=1}
        ]}}}
     Scale Curve=2
    )
    Module.Param0WriteEnabled=true
    Module.Param0WriteEnabled1=true
    Module.Param0WriteEnabled3=true
    Module.Param1WriteEnabled=true
    Module.Param1WriteEnabled1=true
    Module.Param1WriteEnabled3=true
    Module.Param2WriteEnabled=true
    Module.Param2WriteEnabled1=true
    Module.Param2WriteEnabled3=true
    Module.Param3WriteEnabled=true
    Module.Param3WriteEnabled1=true
    Module.Param3WriteEnabled3=true
    Write Parameter Index 1=true
    Write Parameter Index 3=true
   )
   /SetVariables(Particles.DynamicMaterialParam_Numbers:Vector4f=<7 1 0 0>) {isSetParameters=true}
  ]
  rendererClasses [&NiagaraSpriteRendererProperties]
  renderers [
   {rendererClass=&NiagaraSpriteRendererProperties}
   {rendererClass=&NiagaraSpriteRendererProperties, rendererIndex=1}
  ]
 }}
summary {
 systemName=NS_Emote_Hoppin
 emitters [{emitterName=Tiles001, rendererClasses [&NiagaraSpriteRendererProperties]}]
 userVariables [
  User.bIsFrontEnd {type=NiagaraBool}
  User.bIsFrontEndPreview {type=NiagaraBool}
  User.FeetCurve {type=NiagaraFloat, defaultValue=0, description=""}
  User.HopscotchCount {type=NiagaraFloat, defaultValue=0, description=""}
  User.HopscotchSpawnCount {type=NiagaraFloat, defaultValue=0, description=""}
 ]
}
